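; Colored smoke effect Blueprint, named for light posts, that tints its smokestack sprite mesh component with a configurable color.
; The construction script calls its parent, then creates a dynamic material instance on slot 0 and sets EmissiveColor from the Emissive Color variable.
; It also reads Emissive2 from the material cast to a constant instance, but leaves that value unused.
; BeginPlay, Tick and overlap events are empty.
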(event EventBeginPlay)

(event Collision|EventActorBeginOverlap
  (params (OtherActor :type Actor%20Object%20Reference)))

(event EventTick
  (params (DeltaSeconds :type Float%20%28single-precision%29)))

(fn ConstructionScript
  (|CallParentFunction)
  (bind _asmaterial_instance_constant (Utilities|Casting|CastToMaterialInstanceConstant (Rendering|Material|GetMaterial (Variables|Default|GetSM_Smokestack_Sprites_01) 0)))
  (Rendering|Material|GetVectorParameterValue _asmaterial_instance_constant "Emissive2" :!class /Script/Engine.MaterialInstanceConstant)
  (bind _returnvalue (Rendering|Material|CreateDynamicMaterialInstance (Variables|Default|GetSM_Smokestack_Sprites_01) 0 (Rendering|Material|GetMaterial (Variables|Default|GetSM_Smokestack_Sprites_01) 0) "None" :!class /Script/Engine.PrimitiveComponent))
  (Rendering|Material|SetVectorParameterValue _returnvalue "EmissiveColor" (Math|Conversions|ToLinearColor(Color) (Variables|Default|GetEmissive%20Color)) :!class /Script/Engine.MaterialInstanceDynamic))
